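# Niagara system for the Pyrite drift trail vehicle cosmetic, with painted-variant colors in every emitter.
# DriftTrail streams two skeleton-anchored ribbons backward, slower in frontend preview, layering eight renderers with side and top offset positions.
# WheelWrap bursts two persistent wheel-scaled meshes on sockets, while cone spawns short-lived meshes and sprites there with cosine-pulsed color and animated material parameters.
# All three emitters use FN_DisableTireFX for per-wheel masking.
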NiagaraSystem /VehicleCosmetics/DriftTrail/DriftTrail_Pyrite/FX/NS_DriftTrail_Pyrite
dependencies {
 usedDataInterfaces [&NiagaraDataInterfaceCurve, &NiagaraDataInterfaceArrayColor, &NiagaraDataInterfaceSkeletalMesh]
 usedDynamicInputs [&RandomRangeVector]
 usedModules [
  &self:SystemSpawnScript.SystemScriptSource.SystemScriptGraph.NiagaraNodeAssignment_0.SetVariables_#1
  &FN_UnpackVehicleFlags
  /VehicleCosmetics/FX/Niagara/Modules/VC_SetVehicleSpeedAttribs
  &SystemState
  /Game/Effects/Niagara/Modules/Vehicle/FN_VehicleFlags_Cheatcodes
 ]
}
emitters {
 NE_VC_DriftTrail {
  emitterUpdateScript [
   &EmitterState(Inactive Response=::NewEnumerator1, Life Cycle Mode=::NewEnumerator1, Loop Duration=1, UseLoopDelay=false)
   &VC_RateQualityHandler(Epic Spawn Rate=50, High Spawn Rate=40, Low Spawn Rate=30, Use Custom Rate?=true)
   &VC_SpawnRibbon_v1(Number of Ribbons=2, SpawnRate=$Emitter.Spawn Rate)
  ]
  particleSpawnScript [
   &VC_PaintedVariantsOverride(
    Black=<rgba 0.06 0.06 0.06 1>
    BurntSienna=<rgba 0.5 0.1167 0.04 1>
    Cobalt=<rgba 0.024 0.0624 0.6 1>
    Crimson=<rgba 0.6 0.012 0.0316 1>
    DefaultColor=<rgba 0.195 0.3467 0.65 1>
    ForestGreen=<rgba 0 0.5 0 1>
    Grey=<rgba 0.12 0.12 0.12 1>
    Lime=<rgba 0.4 0.6 0 1>
    Orange=<rgba 0.6 0.168 0.024 1>
    Pink=<rgba 0.5 0.02 0.42 1>
    Purple=<rgba 0.1667 0 0.5 1>
    R1 Black:LinearColor=<rgba 0.005 0.005 0.005 1>
    R1 BurntSienna:LinearColor=<rgba 0.008 0.008 0.008 1>
    R1 Cobalt:LinearColor=<rgba 0.012 0.012 0.012 1>
    R1 Crimson:LinearColor=<rgba 0.012 0.012 0.012 1>
    R1 DefaultColor=<rgba 0.012 0.012 0.012 1>
    R1 ForestGreen:LinearColor=<rgba 0.012 0.012 0.012 1>
    R1 Grey:LinearColor=<rgba 0.005 0.005 0.005 1>
    R1 Lime:LinearColor=<rgba 0.01 0.01 0.01 1>
    R1 Orange:LinearColor=<rgba 0.008 0.008 0.008 1>
    R1 Pink:LinearColor=<rgba 0.008 0.008 0.008 1>
    R1 Purple:LinearColor=<rgba 0.01 0.01 0.01 1>
    R1 Saffron:LinearColor=<rgba 0.01 0.01 0.01 1>
    R1 SkyBlue:LinearColor=<rgba 0.01 0.01 0.01 1>
    R1 TitaniumWhite:LinearColor=<rgba 0.012 0.012 0.012 1>
    Saffron=<rgba 0.5 0.375 0 1>
    SkyBlue=<rgba 0 0.3 0.6 1>
    TitaniumWhite=<rgba 0.6 0.6 0.6 1>
    Use R1 Painted Overrides=true
   )
   &VC_InitializeRibbon_v1(
    Color=$Particles.Color
    Custom Offset=true
    Enable Skeletal Position=true
    Lifetime=0.6
    Local Vector Offset=&AddVectorToPosition(Position=<-50 0 -40>, Vector=$User.offset)
    Offset Amount=1
    Ribbon Width=25
   )
   &AddVelocity(Velocity=&MakeCustomVectorFromBool(Bool=$User.bIsFrontendPreview, True Vector=<-2200 0 0>, False Vector=<-2600 0 0>)) {module=AddVelocity001}
   /SetVariables(Particles.leftfacing:Vector3f=<0 1 0>, Particles.topfacing:Vector3f=<0 0 1>) {isSetParameters=true}
  ]
  particleUpdateScript [
   &FN_DisableTireFX(Mask Index Attribute:NiagaraInt32=$Particles.SpawnGroup, Mask Index FL=2, Mask Index FR=3, Mask Index RL=0, Mask Index RR=1)
   &ParticleState()
   &InheritSourceMovement(
    Applied Position Scale=&VectorFromFloat(Value=&FloatFromCurve(FloatCurve=DataInterface() {propertyValues {Curve {keys [<at 0 1 RCIM_Cubic>, {arriveTangent=-1.657, interpMode=RCIM_Cubic, leaveTangent=-1.657, tangentMode=RCTM_User, time=1, value=0}]}}}))
    ApplyForce=false
    ApplyPosition=true
    ApplyPositionFalloff=false
   )
   &SolveForcesAndVelocity(Clamp Velocity=false, Limit Acceleration=false)
   &VC_UpdateRibbon_v1(Enable Skeletal Position=true)
   &ScaleColor(
    Scale Alpha=&FloatFromCurve(FloatCurve=DataInterface() {propertyValues {Curve {keys [
         <at 0 0>
         <at 0.15 1>
         {
          arriveTangentWeight=0.2778
          interpMode=RCIM_Cubic
          leaveTangent=-6.6316
          leaveTangentWeight=0.1389
          tangentMode=RCTM_Break
          time=0.6
          value=1
         }
         {arriveTangentWeight=0.0833, interpMode=RCIM_Cubic, tangentMode=RCTM_User, time=0.95, value=0}
        ]}}})
    Scale RGB=&VectorFromFloat(Value=2)
    ScaleA=true
    ScaleRGB=true
   )
   /SetVariables(
    Particles.colorfill:LinearColor=&MultiplyLinearColor(
     Linear Color=$Particles.R1 Color
     Scale Factor=&MakeVector4FromVectorAndFloat(
      "Vector (RGB)"=&VectorFromFloat(Value=1)
      "Float (Alpha)"=&FloatFromCurve(FloatCurve=DataInterface() {propertyValues {Curve {keys [
           <at 0 0>
           <at 0.1 1>
           {
            arriveTangentWeight=0.2778
            interpMode=RCIM_Cubic
            leaveTangent=-6.6316
            leaveTangentWeight=0.1389
            tangentMode=RCTM_Break
            time=0.6
            value=1
           }
           {arriveTangentWeight=0.0833, interpMode=RCIM_Cubic, tangentMode=RCTM_User, time=0.95, value=0}
          ]}}})
     )
    )
    Particles.widthfill=25
   ) {isSetParameters=true}
   /SetVariables(
    Particles.leftposition:NiagaraPosition=&AddVectorToPosition(Position=$Particles.Position, Vector=&TransformVector(Vector=<0 -8 7.5>, Destination Space=::Simulation))
    Particles.rightposition:NiagaraPosition=&AddVectorToPosition(Position=$Particles.Position, Vector=&TransformVector(Vector=<0 8 7.5>, Destination Space=::Simulation))
    Particles.topposition:NiagaraPosition=&AddVectorToPosition(Position=$Particles.Position, Vector=&TransformVector(Vector=<0 0 15>, Destination Space=::Simulation))
   ) {isSetParameters=true}
  ]
  rendererClasses [&NiagaraRibbonRendererProperties]
  renderers [
   {rendererClass=&NiagaraRibbonRendererProperties}
   {rendererClass=&NiagaraRibbonRendererProperties, rendererIndex=1}
   {rendererClass=&NiagaraRibbonRendererProperties, rendererIndex=2}
   {rendererClass=&NiagaraRibbonRendererProperties, rendererIndex=3}
   {rendererClass=&NiagaraRibbonRendererProperties, rendererIndex=4}
   {rendererClass=&NiagaraRibbonRendererProperties, rendererIndex=5}
   {rendererClass=&NiagaraRibbonRendererProperties, rendererIndex=6}
   {rendererClass=&NiagaraRibbonRendererProperties, rendererIndex=7}
  ]
 }
 NE_VC_WheelWrap {
  emitterUpdateScript [
   &EmitterState(
    Inactive Response=::NewEnumerator1
    Life Cycle Mode=::NewEnumerator1
    Loop Behavior=::NewEnumerator1
    Loop Duration Mode=::NewEnumerator1
    UseLoopDelay=false
   )
   &SpawnBurst_Instantaneous(Spawn Count=2, Use Spawn Probability=false, UseLoopCountLimit=false)
  ]
  particleSpawnScript [
   &VC_PaintedVariantsOverride(
    Black=<rgba 0.13 0.13 0.13 1>
    BurntSienna=<rgba 0.5 0.1775 0.07 1>
    Cobalt=<rgba 0.045 0.1875 0.9 1>
    Crimson=<rgba 0.5 0.085 0.0988 1>
    DefaultColor=<rgba 0.26 0.39 0.65 1>
    ForestGreen=<rgba 0 0.5 0 1>
    Grey=<rgba 0.16 0.16 0.16 1>
    Lime=<rgba 0.3833 0.5 0 1>
    Orange=<rgba 0.5 0.2229 0.025 1>
    Pink=<rgba 0.6 0.024 0.504 1>
    Purple=<rgba 0.25 0 0.5 1>
    R1 Black:LinearColor=<rgba 0.06 0.06 0.06 1>
    R1 BurntSienna:LinearColor=<rgba 0.1 0.1 0.1 1>
    R1 Cobalt:LinearColor=<rgba 0.1 0.1 0.1 1>
    R1 Crimson:LinearColor=<rgba 0.1 0.1 0.1 1>
    R1 DefaultColor=<rgba 0.12 0.12 0.12 1>
    R1 ForestGreen:LinearColor=<rgba 0.11 0.11 0.11 1>
    R1 Gold:LinearColor=<rgba 0.12 0.12 0.12 1>
    R1 Grey:LinearColor=<rgba 0.06 0.06 0.06 1>
    R1 Lime:LinearColor=<rgba 0.06 0.06 0.06 1>
    R1 Orange:LinearColor=<rgba 0.1 0.1 0.1 1>
    R1 Pink:LinearColor=<rgba 0.06 0.06 0.06 1>
    R1 Purple:LinearColor=<rgba 0.08 0.08 0.08 1>
    R1 Saffron:LinearColor=<rgba 0.06 0.06 0.06 1>
    R1 SkyBlue:LinearColor=<rgba 0.12 0.12 0.12 1>
    R1 TitaniumWhite:LinearColor=<rgba 0.12 0.12 0.12 1>
    Saffron=<rgba 0.5 0.375 0 1>
    TitaniumWhite=<rgba 0.5 0.5 0.5 1>
    Use R1 Painted Overrides=true
   )
   &InitializeParticle(
    Color Mode=::NewEnumerator0
    Lifetime=5
    Mesh Scale Mode=::NewEnumerator4
    Sprite Rotation Mode=::NewEnumerator2
    Sprite Size Mode=::NewEnumerator3
    UsePositionOffset=false
    Write Material Random=true
   )
   &PartitionParticles()
   &SkeletalMeshLocation(
    Bone / Socket Index=&Add_Integer(A=$Particles.PartitionParticles.Partition, B=4)
    Bone / Socket Sampling Mode=::NewEnumerator5
    Interpolate Mesh Sampling=false
    MaskSpawnGroup=false
    Mesh Sampling Type=::NewEnumerator6
    Skeletal Mesh=$System.SkeletalMesh
   )
   &InitialMeshOrientation(Mesh Orientation Mode=::NewEnumerator7, Rotation=<0 0 0.5>, Use Rotation Vector=true)
   &Color(Color=<rgba 0 1 0 1>) {enabled=false}
  ]
  particleUpdateScript [
   &FN_DisableTireFX(
    Mask Index Attribute:NiagaraInt32=$Particles.PartitionParticles.Partition
    Mask Index FL=2
    Mask Index FR=3
    Mask Index RL=0
    Mask Index RR=1
   )
   &VC_WheelWrapScale(Wheel Index Attribute:NiagaraInt32=&Add_Integer(A=$Particles.PartitionParticles.Partition, B=2))
   &ParticleState(Kill Particles When Lifetime Has Elapsed=false)
   &UpdateMeshOrientation(Rotation Coordinate Space=::NewEnumerator2, Rotation Vector=<0 1 0>) {enabled=false}
   &SkeletalMeshLocation(
    Bone / Socket Index=&Add_Integer(A=$Particles.PartitionParticles.Partition, B=4)
    Bone / Socket Sampling Mode=::NewEnumerator5
    MaskSpawnGroup=false
    Mesh Sampling Type=::NewEnumerator6
    Skeletal Mesh=$System.SkeletalMesh
   ) {module=SkeletalMeshLocation001}
   &SolveForcesAndVelocity(Clamp Velocity=false, Limit Acceleration=false)
   &ScaleColor(Scale RGB=&VectorFromFloat(Value=2), ScaleA=true, ScaleRGB=true)
  ]
  rendererClasses [&NiagaraMeshRendererProperties]
  renderers [
   {rendererClass=&NiagaraMeshRendererProperties, rendererIndex=0}
   {rendererClass=&NiagaraMeshRendererProperties, rendererIndex=1}
  ]
 }
 NE_VC_cone {
  emitterUpdateScript [
   &EmitterState(Inactive Response=::NewEnumerator1, Life Cycle Mode=::NewEnumerator1, Loop Duration=0.4, UseLoopDelay=false)
   &SpawnRate(SpawnRate=31, Use Spawn Probability=false)
  ]
  particleSpawnScript [
   &VC_PaintedVariantsOverride(
    Black=<rgba 0.08 0.08 0.08 1>
    BurntSienna=<rgba 0.6 0.195 0.06 1>
    Cobalt=<rgba 0.02 0.0833 0.4 1>
    Crimson=<rgba 0.65 0.0325 0.0531 1>
    DefaultColor=&SelectLinearColorFromArray(
     Array Sampling Mode=::NewEnumerator0
     Direct Array Index=?
     Color Selection Array=DataInterface() {propertyValues {ColorData [<rgba 0.315 0.51 0.9 1>, <rgba 0.65 0.2113 0.065 1>]}}
     Random Seed=0
    )
    ForestGreen=<rgba 0 0.65 0 1>
    Grey=<rgba 0.12 0.12 0.12 1>
    Lime=<rgba 0.2667 0.4 0 1>
    Orange=<rgba 0.5 0.1325 0.01 1>
    Pink=<rgba 0.4 0.032 0.3387 1>
    Purple=<rgba 0.1667 0 0.4 1>
    R1 Black:LinearColor=<rgba 0.01 0.01 0.01 2>
    R1 BurntSienna:LinearColor=<rgba 0.15 0.0226 0.003 2>
    R1 Cobalt:LinearColor=<rgba 0 0.01 0.3 1>
    R1 Crimson:LinearColor=<rgba 0.2 0 0 2>
    R1 DefaultColor=&SelectLinearColorFromArray(
     Array Sampling Mode=::NewEnumerator0
     Direct Array Index=?
     Color Selection Array=DataInterface() {propertyValues {ColorData [<rgba 0.04 0.0933 0.2 1.5>, <rgba 0.2 0.065 0.02 1.5>]}}
     Random Seed=0
    )
    R1 ForestGreen:LinearColor=<rgba 0 0.2 0 2>
    R1 Grey:LinearColor=<rgba 0.015 0.015 0.015 2>
    R1 Lime:LinearColor=<rgba 0.075 0.15 0 2>
    R1 Orange:LinearColor=<rgba 0.25 0.0417 0 2>
    R1 Pink:LinearColor=<rgba 0.15 0.003 0.1255 2>
    R1 Purple:LinearColor=<rgba 0.0625 0 0.15 2>
    R1 Saffron:LinearColor=<rgba 0.15 0.1125 0 2>
    R1 SkyBlue:LinearColor=<rgba 0 0.1 0.2 2>
    R1 TitaniumWhite:LinearColor=<rgba 0.1 0.1 0.1 2>
    Saffron=<rgba 0.6 0.45 0 1>
    SkyBlue=<rgba 0.06 0.33 0.6 1>
    TitaniumWhite=<rgba 0.6 0.6 0.6 1>
    Use R1 Painted Overrides=true
   )
   &InitializeParticle(
    Color Mode=::NewEnumerator0
    Lifetime Max=0.5
    Lifetime Min=0.3
    Lifetime Mode=::NewEnumerator1
    Mesh Scale Mode=::NewEnumerator1
    Mesh Uniform Scale Max=0.8
    Mesh Uniform Scale Min=0.5
    UsePositionOffset=false
    Write Material Random=true
   )
   &PartitionParticles()
   &SkeletalMeshLocation(
    Bone / Socket Index=&Add_Integer(A=$Particles.PartitionParticles.Partition, B=4)
    Bone / Socket Sampling Mode=::NewEnumerator5
    Interpolate Mesh Sampling=false
    MaskSpawnGroup=false
    Mesh Sampling Type=::NewEnumerator6
    Skeletal Mesh=$System.SkeletalMesh
   )
   &ShapeLocation(
    MaskSpawnGroup=false
    Non Uniform Scale=<1.3 1 0.65>
    Offset=&AddVectorToPosition(Position=<-46 0 -35>, Vector=$User.offset)
    Offset Coordinate Space=::Local
    Offset Mode=::NewEnumerator0
    Override Seed=false
    Sphere Radius=10
    Sphere Surface Distribution=0.1905
    UseOwnerScale=true
   )
   &AddVelocity(
    Cone Angle=-60
    Cone Axis=<-1 0 0>
    Inner Cone Angle=5
    Module.UseConeFalloff=false
    Module.UseVelDistribution=false
    Override Seed=false
    Velocity Mode=::NewEnumerator2
    Velocity Speed=&RandomRangeFloat(Minimum=100, Maximum=120)
   ) {enabled=false}
   /SetVariables(Particles.R1 Color:LinearColor=&MakeCustomLinearColorFromBool(
     Bool=&SetBoolByIntComparison(A=$User.ColorVariantID, B=0, Comparison Type=::NewEnumerator4)
     True Color=&MultiplyLinearColor(
      Linear Color=$Particles.Color
      Scale Factor=&MakeVector4FromVectorAndFloat("Vector (RGB)"=&VectorFromFloat(Value=0.25), "Float (Alpha)"=1)
     )
     False Color=$Particles.R1 Color
    )) {isSetParameters=true}
  ]
  particleUpdateScript [
   &FN_DisableTireFX(
    Mask Index Attribute:NiagaraInt32=$Particles.PartitionParticles.Partition
    Mask Index FL=2
    Mask Index FR=3
    Mask Index RL=0
    Mask Index RR=1
   )
   &ParticleState()
   &SolveForcesAndVelocity(Clamp Velocity=false, Limit Acceleration=false)
   /SetVariables(
    Particles.coloredge:LinearColor=&MultiplyLinearColor(
     Linear Color=$Particles.Color
     Scale Factor=&MakeVector4FromVectorAndFloat(
      "Vector (RGB)"=&VectorFromFloat(Value=30)
      "Float (Alpha)"=&FloatFromCurve(FloatCurve=DataInterface() {propertyValues {Curve {keys [
           {interpMode=RCIM_Cubic, leaveTangentWeight=0.3333, tangentMode=RCTM_User, time=0, value=0.3}
           {arriveTangent=3.9892, interpMode=RCIM_Cubic, leaveTangent=3.9892, tangentMode=RCTM_User, time=1, value=2}
          ]}}})
     )
    )
    Particles.cubecolor:LinearColor=&MultiplyLinearColor(
     Linear Color=$Particles.R1 Color
     Scale Factor=&MakeVector4FromVectorAndFloat(
      "Vector (RGB)"=&VectorFromFloat(Value=&Cosine(Normalized Angle=$Emitter.Age, Period=0.3, Scale=4, Bias=4.2))
      "Float (Alpha)"=&FloatFromCurve(FloatCurve=DataInterface() {propertyValues {Curve {keys [
           {arriveTangent=16.1644, interpMode=RCIM_Cubic, leaveTangent=16.1644, leaveTangentWeight=0.0333, tangentMode=RCTM_User, time=0, value=0}
           {
            arriveTangent=-3.6686
            arriveTangentWeight=0.0333
            interpMode=RCIM_Cubic
            leaveTangent=-3.6686
            leaveTangentWeight=0.1333
            tangentMode=RCTM_User
            time=0.1
            value=1
           }
           {arriveTangentWeight=0.1648, interpMode=RCIM_Cubic, tangentMode=RCTM_User, time=0.5, value=0}
          ]}}})
     )
    )
   ) {isSetParameters=true}
   &DynamicMaterialParameters(
    Index 0 Param 1=&FloatFromCurve(FloatCurve=DataInterface() {propertyValues {Curve {
        defaultValue=1
        keys [
         {arriveTangent=5.2439, interpMode=RCIM_Cubic, leaveTangent=5.2439, tangentMode=RCTM_User, time=0, value=0.1}
         {arriveTangent=0.8635, interpMode=RCIM_Cubic, leaveTangent=0.8635, leaveTangentWeight=0.3333, tangentMode=RCTM_User, time=0.15, value=1}
         {arriveTangent=-0.9176, arriveTangentWeight=0.3382, interpMode=RCIM_Cubic, leaveTangent=-0.9176, tangentMode=RCTM_User, time=0.95, value=0.15}
        ]
       }}})
    Index 0 Param 2=&FloatFromCurve(
     FloatCurve=DataInterface() {propertyValues {Curve {keys [
         {arriveTangent=0.8711, interpMode=RCIM_Cubic, leaveTangent=0.8711, leaveTangentWeight=0.3333, tangentMode=RCTM_User, time=0, value=0}
         {arriveTangent=0.0311, arriveTangentWeight=0.3333, interpMode=RCIM_Cubic, leaveTangent=0.0311, tangentMode=RCTM_User, time=0.9, value=1}
        ]}}}
     Scale Curve=30
    )
    Module.Index0Float0Write=true
    Module.Index0Float1Write=true
    Module.Index0Float2Write=true
    Module.Index0Float3Write=true
   )
  ]
  rendererClasses [&NiagaraMeshRendererProperties, &NiagaraSpriteRendererProperties]
  renderers [
   {rendererClass=&NiagaraMeshRendererProperties, rendererIndex=0}
   {rendererClass=&NiagaraMeshRendererProperties, rendererIndex=1}
   {rendererClass=&NiagaraSpriteRendererProperties, rendererIndex=2}
  ]
 }
}
summary {
 systemName=NS_DriftTrail_Pyrite
 emitters [
  {emitterName=NE_VC_WheelWrap, rendererClasses [&NiagaraMeshRendererProperties]}
  {emitterName=NE_VC_DriftTrail, rendererClasses [&NiagaraRibbonRendererProperties]}
  {emitterName=NE_VC_cone, rendererClasses [&NiagaraMeshRendererProperties, &NiagaraSpriteRendererProperties]}
 ]
 userVariables [
  User.bIsFrontendPreview {type=NiagaraBool}
  User.bIsPainted {type=NiagaraBool}
  User.ColorVariantID {type=NiagaraInt32}
  User.GameModeID {type=NiagaraInt32}
  User.offset {type=Vector3f, description="offset from wheels(set through cosmetic template)", defaultValue {struct=Vector3f, value=<0 0 0>}}
  User.PaintedColor {type=LinearColor, description="", defaultValue {struct=LinearColor, value=<rgba 1 1 1 1>}}
  User.PaintedScalar {type=NiagaraFloat, defaultValue=1, description=""}
  User.VehicleFlags {type=NiagaraInt32, defaultValue=806289408}
  User.Wheel_Radius_Array {
   type=NiagaraDataInterfaceArrayFloat
   defaultValue {
    struct=NiagaraExt_VariableValue_DataInterface
    value {
     dataInterface=&self:NiagaraDataInterfaceArrayFloat_1
     dataInterfaceClass=&NiagaraDataInterfaceArrayFloat
    }
   }
  }
  User.Wheel_Width_Array {
   type=NiagaraDataInterfaceArrayFloat
   defaultValue {
    struct=NiagaraExt_VariableValue_DataInterface
    value {
     dataInterface=&self:NiagaraDataInterfaceArrayFloat_0
     dataInterfaceClass=&NiagaraDataInterfaceArrayFloat
    }
   }
  }
  User.WheelTouchingArray {
   type=NiagaraDataInterfaceArrayFloat
   defaultValue {
    struct=NiagaraExt_VariableValue_DataInterface
    value {
     dataInterface=&self:NiagaraDataInterfaceArrayFloat_2
     dataInterfaceClass=&NiagaraDataInterfaceArrayFloat
    }
   }
  }
 ]
}
